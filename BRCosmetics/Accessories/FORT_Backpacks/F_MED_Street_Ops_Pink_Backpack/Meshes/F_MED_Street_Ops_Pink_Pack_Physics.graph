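# Physics asset for the Street Ops Pink backpack cosmetic, defining collision and joint limits for its pack, pouches, flap and straps.
# Two boxes form the central pack body, and single boxes cover each of the five dyn_ bones.
# Each dyn_ bone is constrained directly to the pack in a star layout.
# Pouches have tight angular limits, while the flap and especially the straps swing widely on one axis.
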PhysicsAsset /BRCosmetics/Accessories/FORT_Backpacks/F_MED_Street_Ops_Pink_Backpack/Meshes/F_MED_Street_Ops_Pink_Pack_Physics
bodies {
 dyn_pouch_l [{shapeType=Box, extentX=12.0578, extentY=6.3385, extentZ=2.6404, center=<5.0358 -0.1156 1.2782>}]
 dyn_pouch_r [{shapeType=Box, extentX=12.0578, extentY=6.3385, extentZ=2.6404, center=<-5.0358 0.1156 -1.2782>}]
 dyn_satchel_flap [{shapeType=Box, extentX=11.9073, extentY=20.9359, extentZ=0.7256, center=<6.0826 0 0>}]
 dyn_strap_l [{shapeType=Box, extentX=6.8971, extentY=4.0796, extentZ=1.01, center=<2.6426 0.443 -0.1865>}]
 dyn_strap_r [{shapeType=Box, extentX=6.8971, extentY=4.0796, extentZ=1.01, center=<2.6426 -0.2806 -0.1865>}]
 pack [
  {shapeType=Box, extentX=11.0126, extentY=13.4782, extentZ=21.834, center=<-0.0008 -4.29 -6.2591>, rotation=<rot 0 89.9961 -89.9875>}
  {shapeType=Box, extentX=9.4186, extentY=6.3158, extentZ=12.8556, center=<1.9609 -7.9786 -10.3317>, rotation=<rot -16.7274 89.9922 -89.9868>}
 ]
}
constraints [
 {bone1Name=dyn_pouch_r, bone2Name=pack, swing1LimitDegrees=5, swing2LimitDegrees=5, twistLimitDegrees=10}
 {bone1Name=dyn_pouch_l, bone2Name=pack, swing1LimitDegrees=5, swing2LimitDegrees=5, twistLimitDegrees=10}
 {bone1Name=dyn_satchel_flap, bone2Name=pack, swing1LimitDegrees=15, swing2LimitDegrees=41.8571, twistLimitDegrees=5}
 {bone1Name=dyn_strap_l, bone2Name=pack, swing1LimitDegrees=10, swing2LimitDegrees=100, twistLimitDegrees=5}
 {bone1Name=dyn_strap_r, bone2Name=pack, swing1LimitDegrees=10, swing2LimitDegrees=100, twistLimitDegrees=5}
]
